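; AI convoy behavior tree evaluator that decides when an AI pawn should leave its vehicle.
; On enter it caches the playlist's max team size.
; Its starting-condition check runs ordered tests: group leader dead or out of vehicle, vehicle tagged disabled, or tagged vehicle (warbus) low on health, otherwise false.
; The health cutoff is a percentage or absolute value, chosen by a scalable flag and evaluated at team-size level.
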(event EventOnEnter
  (params (OwnerComp :type Behavior%20Tree%20Component%20Object%20Reference))
  (bind (_gamestate _playlist) (Class|FortAsyncActionCurrentPlaylistReady|PlaylistDataReadyAsync)
    (:OnReady
      (Variables|Default|SetMax%20Team%20Size (Variables|/Script/FortniteGame.FortPlaylist|GetMaxTeamSize _playlist)))))

(fn EvaluateStartingConditions
  (params (OwnerComp :type Behavior%20Tree%20Component%20Object%20Reference))
  (Utilities|FlowControl|Sequence
    (:then_0
      ;[1 1 1] Find Group
      (bind _asaicontroller (Utilities|Casting|CastToAIController (Components|GetOwner OwnerComp)))
      (bind _asfort_pawn (Utilities|Casting|CastToFortPawn (Pawn|GetControlledPawn _asaicontroller)))
      (Utilities|IsValid (Actor|GetComponentbyClass _asfort_pawn "/Script/FortniteAI.FortPawnComponent_AIGroup")
        (:Is%20Valid
          ;[1 1 1] If Group Leader is dead or has left vehicle -> Leave
          (Utilities|IsValid (AIGroup|GetGroupLeader (Actor|GetComponentbyClass _asfort_pawn "/Script/FortniteAI.FortPawnComponent_AIGroup"))
            (:Is%20Valid
              (bind _asfort_player_pawn (Utilities|Casting|CastToFortPlayerPawn (Components|GetOwner (AIGroup|GetGroupLeader (Actor|GetComponentbyClass _asfort_pawn "/Script/FortniteAI.FortPawnComponent_AIGroup"))))
                (:then
                  (if (Vehicle|IsinVehicle _asfort_player_pawn)
                    _
                    (else
                      _)))
                (:CastFailed
                  _))
              (return true))
            (:Is%20Not%20Valid
              _)))))
    (:then_1
      ;[1 1 1] Vehicle is disabled
      (bind _asactor (Utilities|Casting|CastToActor (Blackboard|GetObject (Variables|Default|GetVehicle%20Actor) OwnerComp)))
      (bind _returnvalue (GameplayTags|HasAnyMatchingGameplayTags(Message) _asactor (Variables|Default|GetVehicle%20Disabled%20Tag)))
      (if _returnvalue
        (return true)))
    (:then_2
      ;[1 1 1] Vehicle is a warbus and at 50% health
      (bind _asactor_2 (Utilities|Casting|CastToActor (Blackboard|GetObject (Variables|Default|GetVehicle%20Actor) OwnerComp)))
      (bind _returnvalue_2 (GameplayTags|HasAnyMatchingGameplayTags(Message) _asactor_2 (Variables|Default|GetVehicleTag)))
      (if _returnvalue_2
        (bind _asfort_vehicle_interface (Utilities|Casting|CastToFortVehicleInterface _asactor_2))
        (if (> (Class|FortVehicleInterface|GetMaxHealth _asfort_vehicle_interface) 0)
          (if (ScalableFloat|GetValueasBool (Variables|Default|GetUseHealthPercentageForExit))
            (if (< (/ (Class|FortVehicleInterface|GetHealth _asfort_vehicle_interface) (Class|FortVehicleInterface|GetMaxHealth _asfort_vehicle_interface)) (ScalableFloat|GetValueatLevel (Variables|Default|GetHealthPercentageToLeaveVehicle) (Math|Conversions|ToFloat(Integer) (Math|Integer|Max(Integer) (Variables|Default|GetMax%20Team%20Size) 1))))
              (return true))
            (elif (<= (Class|FortVehicleInterface|GetHealth _asfort_vehicle_interface) (ScalableFloat|GetValueatLevel (Variables|Default|GetHealthThresholdToLeaveVehicle) (Math|Conversions|ToFloat(Integer) (Math|Integer|Max(Integer) (Variables|Default|GetMax%20Team%20Size) 1))))
              (return true))))))
    (:then_3
      ;[1 1 1] Don't leave vehicle
      (return false))))
